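; Legacy Lunar Eel small-explosion gameplay actor that delivers an area effect as soon as it spawns.
; On BeginPlay it plays the explosion sound at its location and starts an async targeting query using the Lunar Eel AoE preset.
; When targeting completes, the instigator's ability system asynchronously applies the AoE gameplay effect to the hit results, with a configurable maximum-target value and the instigator as effect context.
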(event EventBeginPlay
  (Audio|PlaySoundatLocation (Variables|Default|GetExplosionSoundCue) (Transformation|GetActorLocation))
  (bind (_asynctaskproxy _targetinghandle) (Class|AsyncActionPerformTargeting|PerformTargetingAsyncAction 0 "/WildEstateWeapons/Gameplay/LunarEel/Legacy/TP_WildEstate_LunarEel_X_AoE.TP_WildEstate_LunarEel_X_AoE" true)
    (:Targeted
      (Ability|Tasks|ApplyGameplayEffectToHitResultsAsync (Ability|GetAbilitySystemComponent (Game|GetInstigator)) (TargetingSystem|TargetingResults|GetTargetingResults (GameInstanceSubsystems|GetTargetingSubsystem) _targetinghandle) (Variables|Default|GetAoEGE) 0 1 (Class|AthenaFunctionLibrary|MakeSimpleFortEffectContext (Game|GetInstigator) (Game|GetInstigator)) (Variables|Default|GetHF_MaxTargets)))))

(fn ConstructionScript)
