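; Legacy LunarEel weapon gameplay ability that turns critical hits into a knockback blast.
; On a critical damage-dealt event it stores the hit, launches the target pawn upward, then briefly delays the environmental blast.
; The blast spawns a cloud actor and sphere-overlaps: players trigger a shockwave verb and non-owners get balloons, while other actors receive vehicle impulses.
; ScalableFloats tune strengths; bullet-count and hit-info helpers are never called here.
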(event Ability|EventActivateAbility
  (bind (_asynctaskproxy _payload) (Ability|Tasks|WaitGameplayEvent "(TagName=\"Event.Damage.Dealt\")")
    (:EventReceived
      (bind _asfort_player_pawn_athena (Utilities|Casting|CastToFortPlayerPawnAthena (.Target (Utilities|Struct|BreakGameplayEventData _payload))))
      (Variables|Default|SetTargetHitPawn _asfort_player_pawn_athena)
      (if (Ability|EffectContext|IsCriticalHit (.ContextHandle (Utilities|Struct|BreakGameplayEventData _payload)))
        (Variables|Default|SetInstigator (.Instigator (Utilities|Struct|BreakGameplayEventData _payload)))
        (Variables|Default|SetHit%20Location (Transformation|GetActorLocation (.Target (Utilities|Struct|BreakGameplayEventData _payload))))
        (Default|DirectHitLaunchHandle)
        (Utilities|FlowControl|Delay 0.05)
        (Default|EnvironmentalKnockback)))))

(event Custom|TryDeployBalloons
  (params (EffectTargetActor :type Actor%20Object%20Reference))
  (if (not (== EffectTargetActor (CallFunction|GetAvatarasFortPlayerPawn)))
    (Utilities|Casting|CastToFortPlayerPawn EffectTargetActor)
    (Ability|SendGameplayEventtoActor EffectTargetActor "(TagName=\"Event.Athena.Balloons.Inflate\")")
    (Ability|ApplyGameplayEffectToTarget (Ability|TargetData|AbilityTargetDatafromActor EffectTargetActor) "/WildEstateWeapons/Gameplay/LunarEel/Legacy/GE_LunarEel_Balloons_Consumable_Passive_B.GE_LunarEel_Balloons_Consumable_Passive_B_C") ; give balloon GE
  ))

(fn CalculateVehicleImpulse
  (params (Object :type Actor%20Object%20Reference))
  (return (Math|Vector|MakeVector (.x (* (Math|Vector|Normalize (- (Transformation|GetActorLocation Object) (Variables|Default|GetHit%20Location))) (ScalableFloat|GetValueatLevel (Variables|Default|GetHFVehicleKnockbackIntensity)))) (.y (* (Math|Vector|Normalize (- (Transformation|GetActorLocation Object) (Variables|Default|GetHit%20Location))) (ScalableFloat|GetValueatLevel (Variables|Default|GetHFVehicleKnockbackIntensity)))) (+ (.z (* (Math|Vector|Normalize (- (Transformation|GetActorLocation Object) (Variables|Default|GetHit%20Location))) (ScalableFloat|GetValueatLevel (Variables|Default|GetHFVehicleKnockbackIntensity)))) (ScalableFloat|GetValueatLevel (Variables|Default|GetHFVehicleZOffsetKnockbackIntensity))))))

(fn CheckAndHandleBulletCount
  (bind _returnvalue (Ability|GetCurrentSourceWeapon))
  (bind _returnvalue_2 (FortWeapon|Ammo|HasInfiniteMagazineAmmo _returnvalue))
  (bind _returnvalue_3 (not _returnvalue_2))
  (bind _returnvalue_4 (FortWeapon|GetBulletsPerClip _returnvalue))
  ; If not using the infinite ammo config
  (if _returnvalue_3
    (Variables|DynamicVariables|SetBulletCount (- (FortWeapon|GetBulletsPerClip (Ability|GetCurrentSourceWeapon)) (FortWeapon|GetLocalMagazineAmmoCount (Ability|GetCurrentSourceWeapon))))
    (elif (and _returnvalue_3 (== (+ (FortWeapon|GetLocalMagazineAmmoCount _returnvalue) 1) _returnvalue_4))
      (Variables|DynamicVariables|SetBulletCount 0)
      ; Check if reached bullets per clip
      (if (>= (Variables|DynamicVariables|GetBulletCount) _returnvalue_4) :!label _label
        (Variables|DynamicVariables|SetBulletCount 1)
        (else
          (bind _result (Math|Integer|IncrementInt (Variables|DynamicVariables|GetBulletCount))
            (:%20%20
              (Variables|DynamicVariables|SetBulletCount _result)))))
      (else
        (goto _label))))
  (return))

(fn DirectHitLaunchHandle
  (Utilities|LaunchPawns (CallFunction|GetAvatarasFortPlayerPawn) (Variables|Default|GetTargetHitPawn) (Math|Vector|MakeVector 0.0 0.0 (ScalableFloat|GetValueatLevel (Variables|Default|GetHFDirectHitLaunchZStrength))) false true true true (!= (ScalableFloat|GetValueatLevel (Variables|Default|GetHFShouldDirectHitLaunchTeammates)) 0) true false false))

(fn EnvironmentalKnockback
  (bind _hit_location (Variables|Default|GetHit%20Location))
  (Class|AthenaFunctionLibrary|CreateEnvironmentalKnockbackExplosion _hit_location (ScalableFloat|GetValueatLevel (Variables|Default|GetHFKnockbackRadius)) (ScalableFloat|GetValueatLevel (Variables|Default|GetHFKnockbackIntensity)) (ScalableFloat|GetValueatLevel (Variables|Default|GetHSMinKnockbackNumber)))
  (Building|SpawnBuildingGameplayActor "/WildEstateWeapons/Gameplay/LunarEel/BGA_Athena_LunarEel_CloudEffect.BGA_Athena_LunarEel_CloudEffect_C" (Math|Transform|MakeTransform _hit_location) (Variables|Default|GetInstigator) true false)
  (bind _outactors (Collision|SphereOverlapActors (Variables|Default|GetHit%20Location) (ScalableFloat|GetValueatLevel (Variables|Default|GetHFKnockbackRadius)) (Variables|Default|GetVehicleFilterObjectType)))
  (for _array_element _outactors
    (if (.bSuccess (Utilities|Casting|CastToFortPlayerPawnAthena _array_element))
      (CustomVerb|BroadcastCustomVerbMessage (CustomVerb|MakeCustomVerbMessage (Variables|Default|GetInstigator) "(TagName=\"CustomVerb.ShockwaveTarget\")"))
      (CallFunction|TryDeployBalloons (Utilities|Casting|CastToFortPlayerPawnAthena _array_element))
      (else
        (bind _impulsevector (CallFunction|CalculateVehicleImpulse _array_element))
        (Utilities|LaunchVehicle _array_element _impulsevector true))))
  (return))

(fn GetHitInfo
  (params (EventData :type Gameplay%20Event%20Data%20Structure))
  (bind (_eventtag _instigator _target _optionalobject _optionalobject2 _contexthandle _instigatortags _targettags _eventmagnitude _targetdata) (Utilities|Struct|BreakGameplayEventData EventData))
  (bind (_bblockinghit _binitialoverlap _time _distance _location _impactpoint _normal _impactnormal _physmat _hitactor) (Collision|BreakHitResult (Ability|TargetData|GetHitResultfromTargetData _targetdata)))
  (Development|PrintString "triggered")
  (return true _location _hitactor))
