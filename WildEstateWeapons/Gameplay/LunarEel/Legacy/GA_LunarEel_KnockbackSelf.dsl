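; Legacy LunarEel weapon ability that knocks its avatar backward, opposite the aim direction, unless blocking tags are present.
; Server-side LaunchPlayer skips vehicle occupants and knockback-immune pawns, and gives grounded crouchers a small fixed backward push.
; Others get jump cancelled and a pitch-offset launch with scalable magnitude and Z adjustments, optionally resetting fall height.
; A delayed velocity check fires a gameplay cue when vertical speed exceeds a threshold, then ends the ability.
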(event Ability|EventActivateAbility
  (bind _returnvalue (GameplayTags|HasAnyMatchingGameplayTags(Message) (Ability|GetAvatarActorfromActorInfo) (Variables|Default|GetTags_PreventKnockback)))
  (if (not _returnvalue)
    (CallFunction|LaunchPlayer)
    ;[1 1 1] Wait a brief moment (Like 2 frames) then check velocity. If Velocity on Z is high enough or low enough play GC.
    (Utilities|Time|SetTimerbyEvent (delegate Custom|CheckVel) (Variables|Default|GetFX%20Delay))
    (else
      (Ability|EndAbility))))

(event Custom|CheckVel
  (bind _returnvalue (Ability|GetAvatarActorfromActorInfo))
  ;[1 1 1] Wait a brief moment (Like 2 frames) then check velocity. If Velocity on Z is high enough or low enough play GC.
  (Utilities|IsValid _returnvalue
    (:Is%20Valid
      (bind _asfort_player_pawn (Utilities|Casting|CastToFortPlayerPawn _returnvalue))
      (Development|PrintString (Utilities|String|ToString(Vector) (Transformation|GetVelocity _asfort_player_pawn)))
      (bind (_returnvalue_x _returnvalue_y _returnvalue_z) (Math|Vector|BreakVector (Transformation|GetVelocity _asfort_player_pawn)))
      (if (> (Math|Float|Absolute(Float) _returnvalue_z) (Variables|Default|GetVelocityForFX))
        (Ability|ExecuteGameplayCueOnOwner (Variables|Default|GetKnockbackCue))
        (else
          _))
      (Ability|EndAbility))))

(event Ability|EventFailedToActivatePassiveAbility
  (params (ActorInfo :type Gameplay%20Ability%20Actor%20Info%20Structure))
  (|CallParentFunction))

(fn LaunchPlayer
  (bind _returnvalue (Ability|GetAvatarActorfromActorInfo))
  (bind _finallaunchvel (Variables|Default|GetFinalLaunchVel))
  ;[1 1 1] Cast to pawn and launch the player a little bit. The distance launcged is based on a max distance from the player who hit them. If the player is further than the max distance than no launch is applied. Otherwise the player is launched based on a ratio of the knock mag using the distance from the player who launched them.
  (if (Networking|IsServer)
    (Utilities|IsValid _returnvalue
      (:Is%20Valid
        (bind _asfort_player_pawn (Utilities|Casting|CastToFortPlayerPawn _returnvalue))
        ;;[1 1 1] Check for Knock Immune or In Vehicle
        (if (not (or (Vehicle|IsinVehicle _asfort_player_pawn) (GameplayTags|HasMatchingGameplayTag _asfort_player_pawn "(TagName=\"Gameplay.Status.KnockbackImmunity\")")))
          ;;[1 1 1] If Crouched do different knock, but not if they are falling, if they are falling do the normal one.
          (if (and (Utilities|Enum|NotEqual(Enum) (Variables|/Script/Engine.CharacterMovementComponent|GetMovementMode (Variables|/Script/Engine.Character|GetCharacterMovement _asfort_player_pawn)) "MOVE_Falling") (Variables|/Script/Engine.Character|GetbIsCrouched _asfort_player_pawn))
            ;;[1 1 1] Launch the player down a little to make the ground splash.
            (Character|LaunchCharacter _asfort_player_pawn (Math|Vector|MakeVector (.x (* (Transformation|GetActorForwardVector _asfort_player_pawn) -1e+02)) (.y (* (Transformation|GetActorForwardVector _asfort_player_pawn) -1e+02))) true true)
            (else
              (FortMovement|CancelJumpAbility _asfort_player_pawn false)
              ;;[1 1 1] Calc Launch Vel
              (Variables|Default|SetFinalLaunchVel (Math|Vector|MakeVector (.x (* (* (Math|Vector|Normalize (Math|Vector|GetForwardVector (Math|Rotator|MakeRotator (.roll (Pawn|GetBaseAimRotation _asfort_player_pawn)) (+ (.pitch (Pawn|GetBaseAimRotation _asfort_player_pawn)) (ScalableFloat|GetValueatLevel (Variables|Default|GetRow_OffsetPitch))) (.yaw (Pawn|GetBaseAimRotation _asfort_player_pawn))))) (ScalableFloat|GetValueatLevel (Variables|Default|GetRow_KnockMag))) -1.0)) (.y (* (* (Math|Vector|Normalize (Math|Vector|GetForwardVector (Math|Rotator|MakeRotator (.roll (Pawn|GetBaseAimRotation _asfort_player_pawn)) (+ (.pitch (Pawn|GetBaseAimRotation _asfort_player_pawn)) (ScalableFloat|GetValueatLevel (Variables|Default|GetRow_OffsetPitch))) (.yaw (Pawn|GetBaseAimRotation _asfort_player_pawn))))) (ScalableFloat|GetValueatLevel (Variables|Default|GetRow_KnockMag))) -1.0)) (Math|Float|float+float (.z (* (* (Math|Vector|Normalize (Math|Vector|GetForwardVector (Math|Rotator|MakeRotator (.roll (Pawn|GetBaseAimRotation _asfort_player_pawn)) (+ (.pitch (Pawn|GetBaseAimRotation _asfort_player_pawn)) (ScalableFloat|GetValueatLevel (Variables|Default|GetRow_OffsetPitch))) (.yaw (Pawn|GetBaseAimRotation _asfort_player_pawn))))) (ScalableFloat|GetValueatLevel (Variables|Default|GetRow_KnockMag))) -1.0)) (ScalableFloat|GetValueatLevel (Variables|Default|GetRow_ZAdd)) (* (.z (Transformation|GetVelocity _asfort_player_pawn)) (ScalableFloat|GetValueatLevel (Variables|Default|GetRow_ZVelMult))))))
              ;;[1 1 1] Stop Movement and launch player
              (Components|Movement|StopMovementImmediately (Variables|/Script/Engine.Character|GetCharacterMovement _asfort_player_pawn))
              (Character|LaunchCharacter _asfort_player_pawn _finallaunchvel true true)
              ;;[1 1 1] If Velocity is Pos Z reset fall height
              (if (and (Math|Conversions|ToBoolean(Integer) (Math|Float|Floor (ScalableFloat|GetValueatLevel (Variables|Default|GetRow_ResetFall)))) (> (.z _finallaunchvel) 0.0))
                (Pawn|Character|ResetFallingHeight _asfort_player_pawn)
                (else
                  _)))))))))
